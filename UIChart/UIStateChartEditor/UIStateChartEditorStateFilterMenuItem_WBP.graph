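# UIStateChartEditorStateFilterMenuItem_WBP is an editor widget for a state filter menu entry in the UI state chart editor.
# It derives from the UIStateChartEditorStateFilterMenuItem class and supplies only the layout: an Overlay root holding the inherited StateNameTextBlock in bold, right-justified text.
# The event graph has only unconnected default events, so any behavior comes from the parent class.
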WidgetBlueprint /UIChart/UIStateChartEditor/UIStateChartEditorStateFilterMenuItem_WBP
graphs [EventGraph {nodes [
   Event PreConstruct()
   Event Construct()
   Event Tick()
  ]}]
variables []
widgets {
 info {
  inheritedWidgetCount=1
  parentClass=/Script/UIChartEditor.UIStateChartEditorStateFilterMenuItem
  rootWidgetClass=&Overlay
  widgetCount=2
 }
 widgets [
  Overlay_0 {widgetClassPath=&Overlay, widget=Overlay_0, properties {visibility=Visible}}
  StateNameTextBlock {
   widgetClassPath=&TextBlock
   bInherited=true
   parent=Overlay_0
   widget=StateNameTextBlock
   properties {
    justification=Right
    font {
     bForceMonospaced=false
     bMaterialIsStencil=false
     fontMaterial=null
     fontObject=&Roboto
     letterSpacing=0
     monospacedWidth=1
     size=12
     skewAmount=0
     typefaceFontName=Bold
     outlineSettings {
      bApplyOutlineToDropShadows=false
      bMiteredCorners=false
      bSeparateFillAlpha=false
      outlineMaterial=null
      outlineSize=0
      outlineColor=<rgba 0 0 0 1>
     }
    }
   }
   slot {class=OverlaySlot, properties {verticalAlignment=VAlign_Fill}}
  }
 ]
}
