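# Editor dialog widget for adding or editing a UI state chart transition, derived from the C++ UIStateChartEditorTransitionAddEditDialog class.
# A details view exposes only the Trigger property, and its change events are forwarded to the parent's handler.
# Source and Target buttons open menu anchors that create state filter menus, registered with the parent via a bIsSource flag.
# Finish and Cancel delegate to parent handlers; a collapsed older Target row remains in the layout.
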WidgetBlueprint /UIChart/UIStateChartEditor/UIStateChartEditorTransitionAddEditDialog_WBP
graphs [
 EventGraph {nodes [
   "On Clicked (FinishButton)"() {is=title K2Node_ComponentBoundEvent}
   > HandleDoneButtonPressed(self:UIState Chart Editor Transition Add Edit Dialog Object Reference)
   "On Clicked (CancelButton)"() {is=title K2Node_ComponentBoundEvent}
   > HandleCancelButtonPressed(self:UIState Chart Editor Transition Add Edit Dialog Object Reference)
   "On Clicked (TargetButton)"() {is=title K2Node_ComponentBoundEvent}
   > Open!(execute, self:Menu Anchor Object Reference=Get TargetMenuAnchor() -> (TargetMenuAnchor:Menu Anchor Object Reference), bFocusMenu:bool=true) -> (then)
   "On Clicked (SourceButton)"() {is=title K2Node_ComponentBoundEvent}
   > Open(self=Get SourceMenuAnchor() -> (SourceMenuAnchor:Menu Anchor Object Reference), bFocusMenu=true)
   n1 = OnTriggerPropertyChanged!() -> (OutputDelegate, then, PropertyName:Name) {is=title K2Node_CustomEvent}
   > HandleDetailsPropertyChanged(self:UIState Chart Editor Transition Add Edit Dialog Object Reference)
   Event Construct()
   > Bind Event to On Property Changed(self:Property View Base Object Reference=Get TriggerPropertyView() -> (TriggerPropertyView:Details View Object Reference), Delegate=@n1) {is=title K2Node_AddDelegate}
   Event PreConstruct()
  ]}
 On_SourceMenuAnchor_GetUserMenuContent_0 {nodes [
   On_SourceMenuAnchor_GetUserMenuContent_0() {is=title K2Node_FunctionEntry}
   > n1 = Create Widget!(
    execute:Exec
    Class:User Widget Class Reference=&/UIStateChartEditorStateFilterMenu_WBP.UIStateChartEditorStateFilterMenu_WBP_C
    OwningPlayer:Player Controller Object Reference=GetOwningPlayer()
   ) -> (then:Exec, ReturnValue:UIState Chart Editor State Filter Menu WBP Object Reference)
   > HandleFilterMenuCreated!(
    execute
    self:UIState Chart Editor Transition Add Edit Dialog Object Reference
    InMenu:UIState Chart Editor State Filter Menu Object Reference=@n1
    bIsSource:bool=true
   ) -> (then)
   > Return Node!(execute, ReturnValue:User Widget Object Reference=@n1)
  ]}
 On_TargetMenuAnchor_GetUserMenuContent_0 {nodes [
   On_TargetMenuAnchor_GetUserMenuContent_0() {is=title K2Node_FunctionEntry}
   > n1 = Create Widget(
    Class=&UIStateChartEditorStateFilterMenu_WBP_C
    OwningPlayer=GetOwningPlayer()
   )
   > HandleFilterMenuCreated(InMenu=@n1, bIsSource=false)
   > Return Node(ReturnValue=@n1)
  ]}
]
variables []
widgets {
 info {
  inheritedWidgetCount=7
  parentClass=/Script/UIChartEditor.UIStateChartEditorTransitionAddEditDialog
  rootWidgetClass=&Overlay
  widgetCount=35
 }
 widgets [
  Overlay_0 {widgetClassPath=&Overlay, widget=Overlay_0}
  Scrim {
   widgetClassPath=&Image
   bIsVariable=true
   parent=Overlay_0
   widget=Scrim
   properties {brush {
     bIsDynamicallyLoaded=false
     drawAs=Image
     imageType=NoImage
     mirroring=NoMirror
     resourceName=None
     resourceObject=null
     tiling=NoTile
     imageSize=<32 32>
     margin=<ltrb 0 0 0 0>
     outlineSettings {
      bUseBrushTransparency=false
      roundingType=HalfHeightRadius
      width=0
      color {colorUseRule=UseColor_Specified, specifiedColor=<rgba 0 0 0 0>}
      cornerRadii=<0 0 0 0>
     }
     tintColor {colorUseRule=UseColor_Specified, specifiedColor=<rgba 0 0 0 0.7>}
     uVRegion {bIsValid=false, max=<0 0>, min=<0 0>}
    }}
   slot {class=OverlaySlot, properties {horizontalAlignment=HAlign_Fill, verticalAlignment=VAlign_Fill}}
  }
  Border_89 {
   widgetClassPath=&Border
   parent=Overlay_0
   widget=Border_89
   properties {brushColor=<rgba 0.0176 0.0176 0.0176 1>, padding=<ltrb 4 20 4 20>}
   slot {class=OverlaySlot, properties {horizontalAlignment=HAlign_Fill, verticalAlignment=VAlign_Center, padding=<ltrb 6 0 6 0>}}
  }
  VerticalBox_48 {
   widgetClassPath=&VerticalBox
   parent=Border_89
   widget=VerticalBox_48
   slot {class=BorderSlot, properties {padding=<ltrb 4 20 4 20>}}
  }
  TriggerRow {widgetClassPath=&HorizontalBox, parent=VerticalBox_48, widget=TriggerRow, slot {class=VerticalBoxSlot}}
  TriggerPropertyView {
   widgetClassPath=&DetailsView
   bInherited=true
   parent=TriggerRow
   widget=TriggerPropertyView
   properties {bShowScrollBar=false, propertiesToShow [Trigger]}
   slot {class=HorizontalBoxSlot, properties {size {sizeRule=Fill, value=1}}}
  }
  Spacer_114 {widgetClassPath=&Spacer, parent=VerticalBox_48, widget=Spacer_114, slot {class=VerticalBoxSlot}}
  SourceRow {widgetClassPath=&HorizontalBox, parent=VerticalBox_48, widget=SourceRow, slot {class=VerticalBoxSlot}}
  SizeBox {
   widgetClassPath=&SizeBox
   parent=SourceRow
   widget=SizeBox
   properties {bOverride_WidthOverride=true, widthOverride=72}
   slot {class=HorizontalBoxSlot, properties {padding=<ltrb 6 0 0 0>}}
  }
  TextBlock {
   widgetClassPath=&TextBlock
   parent=SizeBox
   widget=TextBlock
   properties {
    text=Source
    colorAndOpacity {colorUseRule=UseColor_Specified, specifiedColor=<rgba 0.5765 0.5765 0.5765 1>}
    font {
     bForceMonospaced=false
     bMaterialIsStencil=false
     fontMaterial=null
     fontObject=&Roboto
     letterSpacing=0
     monospacedWidth=1
     size=12
     skewAmount=0
     typefaceFontName=Bold
     outlineSettings {
      bApplyOutlineToDropShadows=false
      bMiteredCorners=false
      bSeparateFillAlpha=false
      outlineMaterial=null
      outlineSize=0
      outlineColor=<rgba 0 0 0 1>
     }
    }
   }
   slot {class=SizeBoxSlot, properties {verticalAlignment=VAlign_Center}}
  }
  SourceMenuAnchor {
   widgetClassPath=/Script/UMG.MenuAnchor
   bInherited=true
   bIsVariable=true
   parent=SourceRow
   widget=SourceMenuAnchor
   slot {class=HorizontalBoxSlot, properties {padding=<ltrb 1 0 1 0>, size {sizeRule=Fill, value=1}}}
  }
  SourceButton {
   widgetClassPath=&Button
   parent=SourceMenuAnchor
   widget=SourceButton
   properties {widgetStyle {
     clickedSlateSound {resourceObject=null}
     disabled {
      bIsDynamicallyLoaded=false
      drawAs=NoDrawType
      imageType=NoImage
      mirroring=NoMirror
      resourceName=None
      resourceObject=null
      tiling=NoTile
      imageSize=<0 0>
      margin=<ltrb 0 0 0 0>
      outlineSettings {
       bUseBrushTransparency=false
       roundingType=HalfHeightRadius
       width=0
       color {colorUseRule=UseColor_Specified, specifiedColor=<rgba 0 0 0 0>}
       cornerRadii=<0 0 0 0>
      }
      tintColor {colorUseRule=UseColor_Specified, specifiedColor=<rgba 1 1 1 1>}
      uVRegion {bIsValid=false, max=<0 0>, min=<0 0>}
     }
     disabledForeground {colorUseRule=UseColor_Specified, specifiedColor=<rgba 0.5271 0.5271 0.5271 1>}
     hovered {
      bIsDynamicallyLoaded=false
      drawAs=RoundedBox
      imageType=NoImage
      mirroring=NoMirror
      resourceName=None
      resourceObject=null
      tiling=NoTile
      imageSize=<32 32>
      margin=<ltrb 0 0 0 0>
      outlineSettings {
       bUseBrushTransparency=true
       roundingType=FixedRadius
       width=1
       color {colorUseRule=UseColor_Specified, specifiedColor=<rgba 0.7243 0.7243 0.7243 1>}
       cornerRadii=<4 4 4 4>
      }
      tintColor {colorUseRule=UseColor_Specified, specifiedColor=<rgba 0.0747 0.0747 0.0747 1>}
      uVRegion {bIsValid=false, max=<0 0>, min=<0 0>}
     }
     hoveredForeground {colorUseRule=UseColor_Specified, specifiedColor=<rgba 0 1 1 1>}
     hoveredSlateSound {resourceObject=null}
     normal {
      bIsDynamicallyLoaded=false
      drawAs=RoundedBox
      imageType=NoImage
      mirroring=NoMirror
      resourceName=None
      resourceObject=null
      tiling=NoTile
      imageSize=<32 32>
      margin=<ltrb 0 0 0 0>
      outlineSettings {
       bUseBrushTransparency=true
       roundingType=FixedRadius
       width=1
       color {colorUseRule=UseColor_Specified, specifiedColor=<rgba 0.6951 0.6951 0.6951 1>}
       cornerRadii=<4 4 4 4>
      }
      tintColor {colorUseRule=UseColor_Specified, specifiedColor=<rgba 0.0314 0.0314 0.0314 1>}
      uVRegion {bIsValid=false, max=<0 0>, min=<0 0>}
     }
     normalForeground {colorUseRule=UseColor_Specified, specifiedColor=<rgba 0 1 1 1>}
     normalPadding=<ltrb 4 1.5 4 1.5>
     pressed {
      bIsDynamicallyLoaded=false
      drawAs=RoundedBox
      imageType=NoImage
      mirroring=NoMirror
      resourceName=None
      resourceObject=null
      tiling=NoTile
      imageSize=<32 32>
      margin=<ltrb 0 0 0 0>
      outlineSettings {
       bUseBrushTransparency=true
       roundingType=FixedRadius
       width=1
       color {colorUseRule=UseColor_Specified, specifiedColor=<rgba 0.7243 0.7243 0.7243 1>}
       cornerRadii=<4 4 4 4>
      }
      tintColor {colorUseRule=UseColor_Specified, specifiedColor=<rgba 0.0314 0.0314 0.0314 1>}
      uVRegion {bIsValid=false, max=<0 0>, min=<0 0>}
     }
     pressedForeground {colorUseRule=UseColor_Specified, specifiedColor=<rgba 0 1 1 1>}
     pressedPadding=<ltrb 4 1.5 4 1.5>
     pressedSlateSound {resourceObject=null}
    }}
   slot {class=PanelSlot}
  }
  SourceTextBlock {
   widgetClassPath=&TextBlock
   bInherited=true
   parent=SourceButton
   widget=SourceTextBlock
   properties {
    clipping=OnDemand
    justification=Right
    colorAndOpacity {colorUseRule=UseColor_Specified, specifiedColor=<rgba 0.9184 0.9184 0.9184 1>}
    font {
     bForceMonospaced=false
     bMaterialIsStencil=false
     fontMaterial=null
     fontObject=&Roboto
     letterSpacing=0
     monospacedWidth=1
     size=12
     skewAmount=0
     typefaceFontName=Bold
     outlineSettings {
      bApplyOutlineToDropShadows=false
      bMiteredCorners=false
      bSeparateFillAlpha=false
      outlineMaterial=null
      outlineSize=0
      outlineColor=<rgba 0 0 0 1>
     }
    }
   }
   slot {class=ButtonSlot, properties {horizontalAlignment=HAlign_Left, padding=<ltrb 0 2 0 2>}}
  }
  TargetRow {widgetClassPath=&HorizontalBox, parent=VerticalBox_48, widget=TargetRow, slot {class=VerticalBoxSlot}}
  SizeBox_1 {
   widgetClassPath=&SizeBox
   parent=TargetRow
   widget=SizeBox_1
   properties {bOverride_WidthOverride=true, widthOverride=72}
   slot {class=HorizontalBoxSlot, properties {padding=<ltrb 6 0 0 0>}}
  }
  TextBlock_2 {
   widgetClassPath=&TextBlock
   parent=SizeBox_1
   widget=TextBlock_2
   properties {
    text=Target
    colorAndOpacity {colorUseRule=UseColor_Specified, specifiedColor=<rgba 0.5765 0.5765 0.5765 1>}
    font {
     bForceMonospaced=false
     bMaterialIsStencil=false
     fontMaterial=null
     fontObject=&Roboto
     letterSpacing=0
     monospacedWidth=1
     size=12
     skewAmount=0
     typefaceFontName=Bold
     outlineSettings {
      bApplyOutlineToDropShadows=false
      bMiteredCorners=false
      bSeparateFillAlpha=false
      outlineMaterial=null
      outlineSize=0
      outlineColor=<rgba 0 0 0 1>
     }
    }
   }
   slot {class=SizeBoxSlot, properties {verticalAlignment=VAlign_Center}}
  }
  TargetMenuAnchor {
   widgetClassPath=&MenuAnchor
   bInherited=true
   bIsVariable=true
   parent=TargetRow
   widget=TargetMenuAnchor
   slot {class=HorizontalBoxSlot, properties {padding=<ltrb 1 0 1 0>, size {sizeRule=Fill, value=1}}}
  }
  TargetButton {
   widgetClassPath=&Button
   parent=TargetMenuAnchor
   widget=TargetButton
   properties {widgetStyle {
     clickedSlateSound {resourceObject=null}
     disabled {
      bIsDynamicallyLoaded=false
      drawAs=NoDrawType
      imageType=NoImage
      mirroring=NoMirror
      resourceName=None
      resourceObject=null
      tiling=NoTile
      imageSize=<0 0>
      margin=<ltrb 0 0 0 0>
      outlineSettings {
       bUseBrushTransparency=false
       roundingType=HalfHeightRadius
       width=0
       color {colorUseRule=UseColor_Specified, specifiedColor=<rgba 0 0 0 0>}
       cornerRadii=<0 0 0 0>
      }
      tintColor {colorUseRule=UseColor_Specified, specifiedColor=<rgba 1 1 1 1>}
      uVRegion {bIsValid=false, max=<0 0>, min=<0 0>}
     }
     disabledForeground {colorUseRule=UseColor_Specified, specifiedColor=<rgba 0.5271 0.5271 0.5271 1>}
     hovered {
      bIsDynamicallyLoaded=false
      drawAs=RoundedBox
      imageType=NoImage
      mirroring=NoMirror
      resourceName=None
      resourceObject=null
      tiling=NoTile
      imageSize=<32 32>
      margin=<ltrb 0 0 0 0>
      outlineSettings {
       bUseBrushTransparency=true
       roundingType=FixedRadius
       width=1
       color {colorUseRule=UseColor_Specified, specifiedColor=<rgba 0.7243 0.7243 0.7243 1>}
       cornerRadii=<4 4 4 4>
      }
      tintColor {colorUseRule=UseColor_Specified, specifiedColor=<rgba 0.0747 0.0747 0.0747 1>}
      uVRegion {bIsValid=false, max=<0 0>, min=<0 0>}
     }
     hoveredForeground {colorUseRule=UseColor_Specified, specifiedColor=<rgba 0 1 1 1>}
     hoveredSlateSound {resourceObject=null}
     normal {
      bIsDynamicallyLoaded=false
      drawAs=RoundedBox
      imageType=NoImage
      mirroring=NoMirror
      resourceName=None
      resourceObject=null
      tiling=NoTile
      imageSize=<32 32>
      margin=<ltrb 0 0 0 0>
      outlineSettings {
       bUseBrushTransparency=true
       roundingType=FixedRadius
       width=1
       color {colorUseRule=UseColor_Specified, specifiedColor=<rgba 0.6951 0.6951 0.6951 1>}
       cornerRadii=<4 4 4 4>
      }
      tintColor {colorUseRule=UseColor_Specified, specifiedColor=<rgba 0.0314 0.0314 0.0314 1>}
      uVRegion {bIsValid=false, max=<0 0>, min=<0 0>}
     }
     normalForeground {colorUseRule=UseColor_Specified, specifiedColor=<rgba 0 1 1 1>}
     normalPadding=<ltrb 4 1.5 4 1.5>
     pressed {
      bIsDynamicallyLoaded=false
      drawAs=RoundedBox
      imageType=NoImage
      mirroring=NoMirror
      resourceName=None
      resourceObject=null
      tiling=NoTile
      imageSize=<32 32>
      margin=<ltrb 0 0 0 0>
      outlineSettings {
       bUseBrushTransparency=true
       roundingType=FixedRadius
       width=1
       color {colorUseRule=UseColor_Specified, specifiedColor=<rgba 0.7243 0.7243 0.7243 1>}
       cornerRadii=<4 4 4 4>
      }
      tintColor {colorUseRule=UseColor_Specified, specifiedColor=<rgba 0.0314 0.0314 0.0314 1>}
      uVRegion {bIsValid=false, max=<0 0>, min=<0 0>}
     }
     pressedForeground {colorUseRule=UseColor_Specified, specifiedColor=<rgba 0 1 1 1>}
     pressedPadding=<ltrb 4 1.5 4 1.5>
     pressedSlateSound {resourceObject=null}
    }}
   slot {class=PanelSlot}
  }
  TargetTextBlock {
   widgetClassPath=&TextBlock
   bInherited=true
   parent=TargetButton
   widget=TargetTextBlock
   properties {
    clipping=OnDemand
    justification=Right
    colorAndOpacity {colorUseRule=UseColor_Specified, specifiedColor=<rgba 0.9184 0.9184 0.9184 1>}
    font {
     bForceMonospaced=false
     bMaterialIsStencil=false
     fontMaterial=null
     fontObject=&Roboto
     letterSpacing=0
     monospacedWidth=1
     size=12
     skewAmount=0
     typefaceFontName=Bold
     outlineSettings {
      bApplyOutlineToDropShadows=false
      bMiteredCorners=false
      bSeparateFillAlpha=false
      outlineMaterial=null
      outlineSize=0
      outlineColor=<rgba 0 0 0 1>
     }
    }
   }
   slot {class=ButtonSlot, properties {horizontalAlignment=HAlign_Left, padding=<ltrb 0 2 0 2>}}
  }
  TargetRowOld {
   widgetClassPath=&HorizontalBox
   parent=VerticalBox_48
   widget=TargetRowOld
   properties {visibility=Collapsed}
   slot {class=VerticalBoxSlot}
  }
  SizeBox_3 {
   widgetClassPath=&SizeBox
   parent=TargetRowOld
   widget=SizeBox_3
   properties {bOverride_WidthOverride=true, widthOverride=72}
   slot {class=HorizontalBoxSlot, properties {padding=<ltrb 6 0 0 0>}}
  }
  TextBlock_1 {
   widgetClassPath=&TextBlock
   parent=SizeBox_3
   widget=TextBlock_1
   properties {
    text=Target
    colorAndOpacity {colorUseRule=UseColor_Specified, specifiedColor=<rgba 0.5765 0.5765 0.5765 1>}
    font {
     bForceMonospaced=false
     bMaterialIsStencil=false
     fontMaterial=null
     fontObject=&Roboto
     letterSpacing=0
     monospacedWidth=1
     size=12
     skewAmount=0
     typefaceFontName=Bold
     outlineSettings {
      bApplyOutlineToDropShadows=false
      bMiteredCorners=false
      bSeparateFillAlpha=false
      outlineMaterial=null
      outlineSize=0
      outlineColor=<rgba 0 0 0 1>
     }
    }
   }
   slot {class=SizeBoxSlot, properties {verticalAlignment=VAlign_Center}}
  }
  Border {
   widgetClassPath=&Border
   parent=TargetRowOld
   widget=Border
   properties {horizontalAlignment=HAlign_Left, brushColor=<rgba 0.0314 0.0314 0.0314 1>}
   slot {class=HorizontalBoxSlot, properties {size {sizeRule=Fill, value=1}}}
  }
  TargetTextBlockOld {
   widgetClassPath=&TextBlock
   parent=Border
   widget=TargetTextBlockOld
   properties {
    clipping=OnDemand
    justification=Right
    colorAndOpacity {colorUseRule=UseColor_Specified, specifiedColor=<rgba 0.9184 0.9184 0.9184 1>}
    font {
     bForceMonospaced=false
     bMaterialIsStencil=false
     fontMaterial=null
     fontObject=&Roboto
     letterSpacing=0
     monospacedWidth=1
     size=12
     skewAmount=0
     typefaceFontName=Bold
     outlineSettings {
      bApplyOutlineToDropShadows=false
      bMiteredCorners=false
      bSeparateFillAlpha=false
      outlineMaterial=null
      outlineSize=0
      outlineColor=<rgba 0 0 0 1>
     }
    }
   }
   slot {class=BorderSlot, properties {horizontalAlignment=HAlign_Left}}
  }
  SizeBox_4 {
   widgetClassPath=&SizeBox
   parent=TargetRowOld
   widget=SizeBox_4
   properties {bOverride_HeightOverride=true, heightOverride=16.7907}
   slot {class=HorizontalBoxSlot}
  }
  TargetMenuAnchorOld {
   widgetClassPath=&MenuAnchor
   bIsVariable=true
   parent=SizeBox_4
   widget=TargetMenuAnchorOld
   properties {placement=MenuPlacement_BelowRightAnchor}
   slot {class=SizeBoxSlot, properties {horizontalAlignment=HAlign_Center, verticalAlignment=VAlign_Center}}
  }
  TargetSearchButton {
   widgetClassPath=&Button
   parent=TargetMenuAnchorOld
   widget=TargetSearchButton
   properties {
    toolTipText=Edit
    widgetStyle {
     clickedSlateSound {resourceObject=null}
     disabled {
      bIsDynamicallyLoaded=false
      drawAs=NoDrawType
      imageType=NoImage
      mirroring=NoMirror
      resourceName=None
      resourceObject=null
      tiling=NoTile
      imageSize=<0 0>
      margin=<ltrb 0 0 0 0>
      outlineSettings {
       bUseBrushTransparency=false
       roundingType=HalfHeightRadius
       width=0
       color {colorUseRule=UseColor_Specified, specifiedColor=<rgba 0 0 0 0>}
       cornerRadii=<0 0 0 0>
      }
      tintColor {colorUseRule=UseColor_Specified, specifiedColor=<rgba 1 1 1 1>}
      uVRegion {bIsValid=false, max=<0 0>, min=<0 0>}
     }
     disabledForeground {colorUseRule=UseColor_Specified, specifiedColor=<rgba 0.5271 0.5271 0.5271 1>}
     hovered {
      bIsDynamicallyLoaded=false
      drawAs=RoundedBox
      imageType=NoImage
      mirroring=NoMirror
      resourceName=None
      resourceObject=null
      tiling=NoTile
      imageSize=<32 32>
      margin=<ltrb 0 0 0 0>
      outlineSettings {
       bUseBrushTransparency=true
       roundingType=FixedRadius
       width=1
       color {colorUseRule=UseColor_Specified, specifiedColor=<rgba 0.7243 0.7243 0.7243 1>}
       cornerRadii=<4 4 4 4>
      }
      tintColor {colorUseRule=UseColor_Specified, specifiedColor=<rgba 0.7243 0.7243 0.7243 1>}
      uVRegion {bIsValid=false, max=<0 0>, min=<0 0>}
     }
     hoveredForeground {colorUseRule=UseColor_Specified, specifiedColor=<rgba 0 1 1 1>}
     hoveredSlateSound {resourceObject=null}
     normal {
      bIsDynamicallyLoaded=false
      drawAs=RoundedBox
      imageType=NoImage
      mirroring=NoMirror
      resourceName=None
      resourceObject=null
      tiling=NoTile
      imageSize=<32 32>
      margin=<ltrb 0 0 0 0>
      outlineSettings {
       bUseBrushTransparency=true
       roundingType=FixedRadius
       width=1
       color {colorUseRule=UseColor_Specified, specifiedColor=<rgba 0.6951 0.6951 0.6951 1>}
       cornerRadii=<4 4 4 4>
      }
      tintColor {colorUseRule=UseColor_Specified, specifiedColor=<rgba 0.4955 0.4955 0.4955 1>}
      uVRegion {bIsValid=false, max=<0 0>, min=<0 0>}
     }
     normalForeground {colorUseRule=UseColor_Specified, specifiedColor=<rgba 0 1 1 1>}
     normalPadding=<ltrb 0 0 0 0>
     pressed {
      bIsDynamicallyLoaded=false
      drawAs=RoundedBox
      imageType=NoImage
      mirroring=NoMirror
      resourceName=None
      resourceObject=null
      tiling=NoTile
      imageSize=<32 32>
      margin=<ltrb 0 0 0 0>
      outlineSettings {
       bUseBrushTransparency=true
       roundingType=FixedRadius
       width=1
       color {colorUseRule=UseColor_Specified, specifiedColor=<rgba 0.7243 0.7243 0.7243 1>}
       cornerRadii=<4 4 4 4>
      }
      tintColor {colorUseRule=UseColor_Specified, specifiedColor=<rgba 0.3843 0.3843 0.3843 1>}
      uVRegion {bIsValid=false, max=<0 0>, min=<0 0>}
     }
     pressedForeground {colorUseRule=UseColor_Specified, specifiedColor=<rgba 0 1 1 1>}
     pressedPadding=<ltrb 0 0 0 0>
     pressedSlateSound {resourceObject=null}
    }
   }
   slot {class=PanelSlot}
  }
  TextBlock_3 {
   widgetClassPath=&TextBlock
   parent=TargetSearchButton
   widget=TextBlock_3
   properties {
    text="✎"
    colorAndOpacity {colorUseRule=UseColor_Specified, specifiedColor=<rgba 0 0 0 1>}
    font {
     bForceMonospaced=false
     bMaterialIsStencil=false
     fontMaterial=null
     fontObject=&Roboto
     letterSpacing=0
     monospacedWidth=1
     size=12
     skewAmount=0
     typefaceFontName=Bold
     outlineSettings {
      bApplyOutlineToDropShadows=false
      bMiteredCorners=false
      bSeparateFillAlpha=false
      outlineMaterial=null
      outlineSize=0
      outlineColor=<rgba 0 0 0 1>
     }
    }
   }
   slot {class=ButtonSlot, properties {padding=<ltrb 1 0 1 0>}}
  }
  Spacer_187 {widgetClassPath=&Spacer, parent=VerticalBox_48, widget=Spacer_187, properties {size=<1 4>}, slot {class=VerticalBoxSlot}}
  HorizontalBox_73 {
   widgetClassPath=&HorizontalBox
   parent=VerticalBox_48
   widget=HorizontalBox_73
   slot {class=VerticalBoxSlot, properties {horizontalAlignment=HAlign_Center}}
  }
  FinishButton {widgetClassPath=&Button, bInherited=true, parent=HorizontalBox_73, widget=FinishButton, slot {class=HorizontalBoxSlot}}
  FinishButtonTextBlock {
   widgetClassPath=&TextBlock
   bInherited=true
   parent=FinishButton
   widget=FinishButtonTextBlock
   properties {
    colorAndOpacity {colorUseRule=UseColor_Specified, specifiedColor=<rgba 0.0851 0.0851 0.0851 1>}
    font {
     bForceMonospaced=false
     bMaterialIsStencil=false
     fontMaterial=null
     fontObject=&Roboto
     letterSpacing=0
     monospacedWidth=1
     size=14
     skewAmount=0
     typefaceFontName=Bold
     outlineSettings {
      bApplyOutlineToDropShadows=false
      bMiteredCorners=false
      bSeparateFillAlpha=false
      outlineMaterial=null
      outlineSize=0
      outlineColor=<rgba 0 0 0 1>
     }
    }
   }
   slot {class=ButtonSlot}
  }
  Spacer_91 {widgetClassPath=&Spacer, parent=HorizontalBox_73, widget=Spacer_91, properties {size=<4 1>}, slot {class=HorizontalBoxSlot}}
  CancelButton {widgetClassPath=&Button, parent=HorizontalBox_73, widget=CancelButton, slot {class=HorizontalBoxSlot}}
  TextBlock_4 {
   widgetClassPath=&TextBlock
   parent=CancelButton
   widget=TextBlock_4
   properties {
    text=Cancel
    colorAndOpacity {colorUseRule=UseColor_Specified, specifiedColor=<rgba 0.0851 0.0851 0.0851 1>}
    font {
     bForceMonospaced=false
     bMaterialIsStencil=false
     fontMaterial=null
     fontObject=&Roboto
     letterSpacing=0
     monospacedWidth=1
     size=14
     skewAmount=0
     typefaceFontName=Bold
     outlineSettings {
      bApplyOutlineToDropShadows=false
      bMiteredCorners=false
      bSeparateFillAlpha=false
      outlineMaterial=null
      outlineSize=0
      outlineColor=<rgba 0 0 0 1>
     }
    }
   }
   slot {class=ButtonSlot}
  }
 ]
}
