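# Editor widget for the UI state chart editor's state context menu, a fixed-width vertical list of buttons.
# Buttons add child, parent or parallel region, move up or down, convert between region and regular child, delete and preview scenes.
# Each click event calls the matching handler on the UIStateChartEditorStateContextMenu parent class, so the graph holds no logic of its own.
# Several buttons are inherited widgets from that parent; PreConstruct, Construct and Tick are empty.
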WidgetBlueprint /UIChart/UIStateChartEditor/UIStateChartEditorStateContextMenu_WBP
graphs [EventGraph {nodes [
   "On Clicked (AddParallelRegionButton)"() {is=title K2Node_ComponentBoundEvent}
   > HandleCreateParallelRegionRequested(self:UIState Chart Editor State Context Menu Object Reference)
   "On Clicked (ConvertToParallelRegionButton)"() {is=title K2Node_ComponentBoundEvent}
   > HandleConvertToParallelRegionRequested(self:UIState Chart Editor State Context Menu Object Reference)
   "On Clicked (ConvertToRegularChildButton)"() {is=title K2Node_ComponentBoundEvent}
   > HandleConvertToRegularChildRequested(self:UIState Chart Editor State Context Menu Object Reference)
   "On Clicked (PreviewButton)"() {is=title K2Node_ComponentBoundEvent}
   > HandlePreviewRequested(self:UIState Chart Editor State Context Menu Object Reference)
   "On Clicked (AddChildButton)"() {is=title K2Node_ComponentBoundEvent}
   > HandleCreateChildRequested(self:UIState Chart Editor State Context Menu Object Reference)
   "On Clicked (AddParentButton)"() {is=title K2Node_ComponentBoundEvent}
   > HandleCreateParentRequested(self:UIState Chart Editor State Context Menu Object Reference)
   "On Clicked (MoveUpButton)"() {is=title K2Node_ComponentBoundEvent}
   > HandleMoveUpRequested(self:UIState Chart Editor State Context Menu Object Reference)
   "On Clicked (MoveDownButton)"() {is=title K2Node_ComponentBoundEvent}
   > HandleMoveDownRequested(self:UIState Chart Editor State Context Menu Object Reference)
   "On Clicked (DeleteButton)"() {is=title K2Node_ComponentBoundEvent}
   > HandleDeleteRequested(self:UIState Chart Editor State Context Menu Object Reference)
   Event PreConstruct()
   Event Construct()
   Event Tick()
  ]}]
variables []
widgets {
 info {
  inheritedWidgetCount=6
  parentClass=/Script/UIChartEditor.UIStateChartEditorStateContextMenu
  rootWidgetClass=&Border
  widgetCount=22
 }
 widgets [
  Border_0 {
   widgetClassPath=&Border
   widget=Border_0
   properties {background {
     bIsDynamicallyLoaded=false
     drawAs=Image
     imageType=NoImage
     mirroring=NoMirror
     resourceName=None
     resourceObject=null
     tiling=NoTile
     imageSize=<32 32>
     margin=<ltrb 0 0 0 0>
     outlineSettings {
      bUseBrushTransparency=false
      roundingType=HalfHeightRadius
      width=0
      color {colorUseRule=UseColor_Specified, specifiedColor=<rgba 0 0 0 0>}
      cornerRadii=<0 0 0 0>
     }
     tintColor {colorUseRule=UseColor_Specified, specifiedColor=<rgba 0.1233 0.1233 0.1233 1>}
     uVRegion {bIsValid=false, max=<0 0>, min=<0 0>}
    }}
  }
  SizeBox_36 {
   widgetClassPath=&SizeBox
   parent=Border_0
   widget=SizeBox_36
   properties {bOverride_WidthOverride=true, widthOverride=200}
   slot {class=BorderSlot}
  }
  VerticalBox_47 {widgetClassPath=&VerticalBox, parent=SizeBox_36, widget=VerticalBox_47, slot {class=SizeBoxSlot}}
  AddChildButton {widgetClassPath=&Button, parent=VerticalBox_47, widget=AddChildButton, slot {class=VerticalBoxSlot}}
  TextBlock_43 {
   widgetClassPath=&TextBlock
   parent=AddChildButton
   widget=TextBlock_43
   properties {
    text=+ Add Child
    colorAndOpacity {colorUseRule=UseColor_Specified, specifiedColor=<rgba 0 0 0 1>}
    font {
     bForceMonospaced=false
     bMaterialIsStencil=false
     fontMaterial=null
     fontObject=&Roboto
     letterSpacing=0
     monospacedWidth=1
     size=12
     skewAmount=0
     typefaceFontName=Bold
     outlineSettings {
      bApplyOutlineToDropShadows=false
      bMiteredCorners=false
      bSeparateFillAlpha=false
      outlineMaterial=null
      outlineSize=0
      outlineColor=<rgba 0 0 0 1>
     }
    }
   }
   slot {class=ButtonSlot, properties {horizontalAlignment=HAlign_Fill}}
  }
  AddParallelRegionButton {widgetClassPath=&Button, parent=VerticalBox_47, widget=AddParallelRegionButton, slot {class=VerticalBoxSlot}}
  TextBlock_4 {
   widgetClassPath=&TextBlock
   parent=AddParallelRegionButton
   widget=TextBlock_4
   properties {
    text=+ Add Parallel Region
    colorAndOpacity {colorUseRule=UseColor_Specified, specifiedColor=<rgba 0 0 0 1>}
    font {
     bForceMonospaced=false
     bMaterialIsStencil=false
     fontMaterial=null
     fontObject=&Roboto
     letterSpacing=0
     monospacedWidth=1
     size=12
     skewAmount=0
     typefaceFontName=Bold
     outlineSettings {
      bApplyOutlineToDropShadows=false
      bMiteredCorners=false
      bSeparateFillAlpha=false
      outlineMaterial=null
      outlineSize=0
      outlineColor=<rgba 0 0 0 1>
     }
    }
   }
   slot {class=ButtonSlot, properties {horizontalAlignment=HAlign_Fill}}
  }
  AddParentButton {widgetClassPath=&Button, parent=VerticalBox_47, widget=AddParentButton, slot {class=VerticalBoxSlot}}
  TextBlock {
   widgetClassPath=&TextBlock
   parent=AddParentButton
   widget=TextBlock
   properties {
    text="^ Add Parent"
    colorAndOpacity {colorUseRule=UseColor_Specified, specifiedColor=<rgba 0 0 0 1>}
    font {
     bForceMonospaced=false
     bMaterialIsStencil=false
     fontMaterial=null
     fontObject=&Roboto
     letterSpacing=0
     monospacedWidth=1
     size=12
     skewAmount=0
     typefaceFontName=Bold
     outlineSettings {
      bApplyOutlineToDropShadows=false
      bMiteredCorners=false
      bSeparateFillAlpha=false
      outlineMaterial=null
      outlineSize=0
      outlineColor=<rgba 0 0 0 1>
     }
    }
   }
   slot {class=ButtonSlot, properties {horizontalAlignment=HAlign_Fill}}
  }
  MoveUpButton {widgetClassPath=&Button, bInherited=true, parent=VerticalBox_47, widget=MoveUpButton, slot {class=VerticalBoxSlot}}
  TextBlock_2 {
   widgetClassPath=&TextBlock
   parent=MoveUpButton
   widget=TextBlock_2
   properties {
    text="↑ Move Up"
    colorAndOpacity {colorUseRule=UseColor_Specified, specifiedColor=<rgba 0 0 0 1>}
    font {
     bForceMonospaced=false
     bMaterialIsStencil=false
     fontMaterial=null
     fontObject=&Roboto
     letterSpacing=0
     monospacedWidth=1
     size=12
     skewAmount=0
     typefaceFontName=Bold
     outlineSettings {
      bApplyOutlineToDropShadows=false
      bMiteredCorners=false
      bSeparateFillAlpha=false
      outlineMaterial=null
      outlineSize=0
      outlineColor=<rgba 0 0 0 1>
     }
    }
   }
   slot {class=ButtonSlot, properties {horizontalAlignment=HAlign_Fill}}
  }
  MoveDownButton {widgetClassPath=&Button, bInherited=true, parent=VerticalBox_47, widget=MoveDownButton, slot {class=VerticalBoxSlot}}
  TextBlock_3 {
   widgetClassPath=&TextBlock
   parent=MoveDownButton
   widget=TextBlock_3
   properties {
    text="↓ Move Down"
    colorAndOpacity {colorUseRule=UseColor_Specified, specifiedColor=<rgba 0 0 0 1>}
    font {
     bForceMonospaced=false
     bMaterialIsStencil=false
     fontMaterial=null
     fontObject=&Roboto
     letterSpacing=0
     monospacedWidth=1
     size=12
     skewAmount=0
     typefaceFontName=Bold
     outlineSettings {
      bApplyOutlineToDropShadows=false
      bMiteredCorners=false
      bSeparateFillAlpha=false
      outlineMaterial=null
      outlineSize=0
      outlineColor=<rgba 0 0 0 1>
     }
    }
   }
   slot {class=ButtonSlot, properties {horizontalAlignment=HAlign_Fill}}
  }
  ConvertToParallelRegionButton {
   widgetClassPath=&Button
   bInherited=true
   parent=VerticalBox_47
   widget=ConvertToParallelRegionButton
   slot {class=VerticalBoxSlot}
  }
  TextBlock_5 {
   widgetClassPath=&TextBlock
   parent=ConvertToParallelRegionButton
   widget=TextBlock_5
   properties {
    text="↔ Convert To Parallel Region"
    colorAndOpacity {colorUseRule=UseColor_Specified, specifiedColor=<rgba 0 0 0 1>}
    font {
     bForceMonospaced=false
     bMaterialIsStencil=false
     fontMaterial=null
     fontObject=&Roboto
     letterSpacing=0
     monospacedWidth=1
     size=12
     skewAmount=0
     typefaceFontName=Bold
     outlineSettings {
      bApplyOutlineToDropShadows=false
      bMiteredCorners=false
      bSeparateFillAlpha=false
      outlineMaterial=null
      outlineSize=0
      outlineColor=<rgba 0 0 0 1>
     }
    }
   }
   slot {class=ButtonSlot, properties {horizontalAlignment=HAlign_Fill}}
  }
  ConvertToRegularChildButton {widgetClassPath=&Button, bInherited=true, parent=VerticalBox_47, widget=ConvertToRegularChildButton, slot {class=VerticalBoxSlot}}
  TextBlock_6 {
   widgetClassPath=&TextBlock
   parent=ConvertToRegularChildButton
   widget=TextBlock_6
   properties {
    text="↔ Convert To Regular Child"
    colorAndOpacity {colorUseRule=UseColor_Specified, specifiedColor=<rgba 0 0 0 1>}
    font {
     bForceMonospaced=false
     bMaterialIsStencil=false
     fontMaterial=null
     fontObject=&Roboto
     letterSpacing=0
     monospacedWidth=1
     size=12
     skewAmount=0
     typefaceFontName=Bold
     outlineSettings {
      bApplyOutlineToDropShadows=false
      bMiteredCorners=false
      bSeparateFillAlpha=false
      outlineMaterial=null
      outlineSize=0
      outlineColor=<rgba 0 0 0 1>
     }
    }
   }
   slot {class=ButtonSlot, properties {horizontalAlignment=HAlign_Fill}}
  }
  DeleteButton {widgetClassPath=&Button, bInherited=true, parent=VerticalBox_47, widget=DeleteButton, slot {class=VerticalBoxSlot}}
  TextBlock_1 {
   widgetClassPath=&TextBlock
   parent=DeleteButton
   widget=TextBlock_1
   properties {
    text="✖ Delete"
    colorAndOpacity {colorUseRule=UseColor_Specified, specifiedColor=<rgba 0 0 0 1>}
    font {
     bForceMonospaced=false
     bMaterialIsStencil=false
     fontMaterial=null
     fontObject=&Roboto
     letterSpacing=0
     monospacedWidth=1
     size=12
     skewAmount=0
     typefaceFontName=Bold
     outlineSettings {
      bApplyOutlineToDropShadows=false
      bMiteredCorners=false
      bSeparateFillAlpha=false
      outlineMaterial=null
      outlineSize=0
      outlineColor=<rgba 0 0 0 1>
     }
    }
   }
   slot {class=ButtonSlot, properties {horizontalAlignment=HAlign_Fill}}
  }
  Spacer_171 {widgetClassPath=&Spacer, parent=VerticalBox_47, widget=Spacer_171, properties {size=<1 4>}, slot {class=VerticalBoxSlot}}
  PreviewButton {widgetClassPath=&Button, bInherited=true, parent=VerticalBox_47, widget=PreviewButton, slot {class=VerticalBoxSlot}}
  TextBlock_7 {
   widgetClassPath=&TextBlock
   parent=PreviewButton
   widget=TextBlock_7
   properties {
    text="□ Preview Scenes"
    colorAndOpacity {colorUseRule=UseColor_Specified, specifiedColor=<rgba 0 0 0 1>}
    font {
     bForceMonospaced=false
     bMaterialIsStencil=false
     fontMaterial=null
     fontObject=&Roboto
     letterSpacing=0
     monospacedWidth=1
     size=12
     skewAmount=0
     typefaceFontName=Bold
     outlineSettings {
      bApplyOutlineToDropShadows=false
      bMiteredCorners=false
      bSeparateFillAlpha=false
      outlineMaterial=null
      outlineSize=0
      outlineColor=<rgba 0 0 0 1>
     }
    }
   }
   slot {class=ButtonSlot, properties {horizontalAlignment=HAlign_Fill}}
  }
 ]
}
